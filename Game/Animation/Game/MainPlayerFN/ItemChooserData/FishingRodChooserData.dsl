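; Chooser data for the fishing rod (FloppingRabbit) in the main player's animation setup, exposing the rod's grip state for animation selection.
; On each chooser data update, it gets the current weapon and, if valid, casts it to the FloppingRabbit weapon Blueprint.
; It then copies that weapon's OneHandGrip value into the bOneHandGrip chooser field.
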(event EventBlueprintUpdateChooserData
  (params (DeltaTime :type Float%20%28single-precision%29))
  (bind _returnvalue (CallFunction|TryGetWeapon))
  (if (Utilities|IsValid _returnvalue :!class /Script/Engine.KismetSystemLibrary)
    (bind _asb_flopping_rabbit_weap_athena (Utilities|Casting|CastToB_FloppingRabbit_Weap_Athena _returnvalue))
    (Variables|FortFloppingRabbitChooserData|SetbOneHandGrip (Variables|/Game/Athena/Items/Consumables/FloppingRabbit/B_FloppingRabbit_Weap_Athena.B_FloppingRabbit_Weap_Athena_C|GetOneHandGrip _asb_flopping_rabbit_weap_athena))))
